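# Masked, two-sided skeletal-mesh material for the sawblades of the Wayfare Mask Female pickaxe, scrolling their textures like a running chainsaw.
# One Panner, driven along U by the PanSpeed_chainsaw parameter, supplies UVs to every texture sample.
# The samples read the sawblade diffuse, normal and specular mask textures, with diffuse and normal each sampled twice.
# The graph shows no connections from these samples to material outputs.
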Material /BRCosmetics/Weapons/FORT_Melee/Pickaxe_Wayfare_Mask_Female/Materials/M_Pickaxe_Wayfare_Mask_Female_blades
graph {
 properties {
  bUsedWithMeshDeformer=true
  bUsedWithSkeletalMesh=true
  blendMode=BLEND_Masked
  pixelDepthOffsetMode=PDOM_Legacy
  twoSided=true
  usedShadingModels=Default Lit
 }
 nodes [
  n1 = Panner(Coordinate=TextureCoordinate(), Speed=AppendVector(A=ScalarParameter() {defaultValue=-11, parameterName=PanSpeed_chainsaw}, B=Constant()))
  TextureSample(UVs=@n1) {texture=/BRCosmetics/Weapons/FORT_Melee/Pickaxe_Wayfare_Mask_Female/Textures/T_Pickaxe_Wayfare_Mask_Female_SAWBLADE_D}
  TextureSample(UVs=@n1) {
   samplerType=SAMPLERTYPE_Normal
   texture=/BRCosmetics/Weapons/FORT_Melee/Pickaxe_Wayfare_Mask_Female/Textures/T_Pickaxe_Wayfare_Mask_Female_SAWBLADE_N
  }
  TextureSample(UVs=@n1) {
   samplerType=SAMPLERTYPE_Masks
   texture=/BRCosmetics/Weapons/FORT_Melee/Pickaxe_Wayfare_Mask_Female/Textures/T_Pickaxe_Wayfare_Mask_Female_SAWBLADE_S
  }
  TextureSample(UVs=@n1) {texture=&T_Pickaxe_Wayfare_Mask_Female_SAWBLADE_D}
  TextureSample(UVs=@n1) {
   samplerType=SAMPLERTYPE_Normal
   texture=&T_Pickaxe_Wayfare_Mask_Female_SAWBLADE_N
  }
 ]
}
